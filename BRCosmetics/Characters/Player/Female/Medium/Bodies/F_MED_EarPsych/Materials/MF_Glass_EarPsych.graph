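# Material function that builds a parameterized glass surface for the EarPsych female character body, assembled as material attributes, with base color, SRM and normal outputs.
# Static switches optionally bring in diffuse (UV0/UV1 blend), emissive and specular/roughness/metallic maps, blended with scalar or color parameters.
# A thickness-scaled Fresnel term supplies view-dependent transparency between inner and outer values.
# An optional channel-masked opacity, with a TechArtMask input available, feeds both Opacity and OpacityMask.
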MaterialFunction /BRCosmetics/Characters/Player/Female/Medium/Bodies/F_MED_EarPsych/Materials/MF_Glass_EarPsych
graph {nodes [
  Constant()
  FunctionOutput(None=NamedRerouteUsage())
  n1 = TextureSampleParameter2D(UVs=LinearInterpolate(
    A=TextureCoordinate()
    B=TextureCoordinate() {coordinateIndex=1}
    Alpha=ScalarParameter() {group=00_Glass, parameterName=use 2nd UV for Diffuse Texture}
   )) {
   group=00_Glass
   parameterName=Diffuse
   samplerSource=SSM_Wrap_WorldGroupSettings
   sortPriority=-100
   texture=/Game/Global/Textures/Default/Blanks/T_MidgreysRGB
  }
  FunctionOutput(None=@n1.RGB) {outputName=BaseColor}
  n2 = TextureSampleParameter2D() {
   desc="R = Spec , G = Roughness, B= Metal"
   group=00_Glass
   parameterName=SRM
   samplerSource=SSM_Wrap_WorldGroupSettings
   samplerType=SAMPLERTYPE_Masks
   sortPriority=-99
   texture=&T_Fortnite_Default_S
  }
  FunctionOutput(None=@n2.RGB) {outputName=SRM}
  n3 = TextureSampleParameter2D() {
   group=00_Glass
   parameterName=Normals
   samplerSource=SSM_Wrap_WorldGroupSettings
   samplerType=SAMPLERTYPE_Normal
   sortPriority=-98
   texture=&FlatNormalBlue
  }
  FunctionOutput(None=NamedRerouteDeclaration(None=@n3.RGB) {name=GlassNormal, nodeColor=<rgba 0.8047 0 1 1>}) {outputName=Normal}
  n4 = VectorParameter() {group=00_Glass, parameterName=GlassDiffuse, sortPriority=1, defaultValue=<rgba 0 0 0 1>}
  n5 = ScalarParameter() {defaultValue=1, group=00_Glass, parameterName=DiffuseTextureBlend, sortPriority=10}
  NamedRerouteDeclaration(None=StaticSwitchParameter(True=LinearInterpolate(A=@n4.RGB, B=@n1.RGB, Alpha=@n5), False=@n4.RGB) {group=00_Glass, parameterName=useDiffuseMap, sortPriority=-50}) {name=GlassBaseColor, nodeColor=<rgba 0 1 0.0625 1>}
  n6 = TextureSampleParameter2D() {
   group=00_Glass
   parameterName=Emissive
   samplerSource=SSM_Wrap_WorldGroupSettings
   sortPriority=-97
   texture=&T_BlacksRGB
  }
  n7 = ScalarParameter() {group=00_Glass, parameterName=EmissiveStrength, sortPriority=15}
  NamedRerouteDeclaration(None=StaticSwitchParameter(True=Multiply(A=@n6.RGB, B=@n7), False=Constant()) {group=00_Glass, parameterName=useEmissiveMap, sortPriority=-48}) {name=GlassEmissive, nodeColor=<rgba 0 0.9062 1 1>}
  n8 = ScalarParameter() {defaultValue=1, group=00_Glass, parameterName=GlassSpecular, sortPriority=2}
  n9 = LinearInterpolate(A=@n8, B=NamedRerouteUsage(), Alpha=ScalarParameter() {defaultValue=1, group=00_Glass, parameterName=SpecularTextureBlend, sortPriority=3})
  n10 = ScalarParameter() {group=00_Glass, parameterName=GlassRoughness, sortPriority=4}
  n11 = LinearInterpolate(A=@n10, B=NamedRerouteUsage(), Alpha=ScalarParameter() {defaultValue=1, group=00_Glass, parameterName=RoughnessTextureBlend, sortPriority=5})
  n12 = ScalarParameter() {defaultValue=0.1, group=00_Glass, parameterName=GlassMetallic, sortPriority=6}
  n13 = LinearInterpolate(A=@n12, B=NamedRerouteUsage(), Alpha=ScalarParameter() {defaultValue=1, group=00_Glass, parameterName=MetallicTextureBlend, sortPriority=7})
  n14 = MaterialFunctionCall(Float3=StaticSwitchParameter(
    True=MaterialFunctionCall(X=@n9, Y=@n11, Z=@n13) {materialFunction=&MakeFloat3}.Result
    False=MaterialFunctionCall(X=@n8, Y=@n10, Z=@n12) {materialFunction=&MakeFloat3}.Result
   ) {group=00_Glass, parameterName=useSRM, sortPriority=-49}) {materialFunction=&BreakOutFloat3Components}
  NamedRerouteDeclaration(None=@n14.R) {name=GlassSpecular, nodeColor=<rgba 0 1 0.0625 1>}
  NamedRerouteDeclaration(None=@n14.B) {name=GlassMetallic, nodeColor=<rgba 0 1 0.0625 1>}
  NamedRerouteDeclaration(None=@n14.G) {name=GlassRoughness, nodeColor=<rgba 0 1 0.0625 1>}
  n15 = NamedRerouteUsage()
  n16 = SetMaterialAttributes!(
   MaterialAttributes
   BaseColor=NamedRerouteUsage()
   Opacity=@n15
   OpacityMask=@n15
   EmissiveColor=NamedRerouteUsage()
   Metallic=NamedRerouteUsage()
   Specular=NamedRerouteUsage()
   Roughness=NamedRerouteUsage()
   Normal=NamedRerouteUsage()
  ) {attributeSetTypes [
    #1
    #2
    #3
    #4
    #5
    #6
    #7
    #8
   ]}
  NamedRerouteDeclaration(None=@n16) {name=GlassMaterial, nodeColor=<rgba 0.8516 0 1 1>}
  n17 = Fresnel(ExponentIn=ScalarParameter() {defaultValue=1, group=00_Glass, parameterName=FresnelExponentTransparency, sortPriority=30})
  n18 = ScalarParameter() {defaultValue=1, group=00_Glass, parameterName=GlassThickness, sortPriority=31}
  NamedRerouteDeclaration(None=Saturate(None=Multiply(A=@n17, B=@n18))) {name=GlassFresnel, nodeColor=<rgba 1 0.1875 0 1>}
  NamedRerouteDeclaration(None=@n2.R) {name=TextureSpecular, nodeColor=<rgba 1 0.5859 0 1>}
  NamedRerouteDeclaration(None=@n2.G) {name=TextureRoughness, nodeColor=<rgba 1 0.5859 0 1>}
  NamedRerouteDeclaration(None=@n2.B) {name=TextureMetallic, nodeColor=<rgba 1 0.5859 0 1>}
  NamedRerouteDeclaration(None=LinearInterpolate(
    A=ScalarParameter() {group=00_Glass, parameterName=InnerTransparency}
    B=ScalarParameter() {defaultValue=0.85, group=00_Glass, parameterName=OuterTransparency, sortPriority=33}
    Alpha=NamedRerouteUsage()
   )) {name=GlassFresnelTransparency, nodeColor=<rgba 0 1 0.9531 1>}
  NamedRerouteDeclaration(None=FunctionInput() {bUsePreviewValueAsDefault=true, inputName=TechArtMask, previewValue=<1 1 1 1>}) {name=TechArtMask, nodeColor=<rgba 0.8047 0 1 1>}
  NamedRerouteDeclaration(None=ChannelMaskParameter(None=NamedRerouteUsage()) {group=00_Glass, parameterName=TextureOpacityChannel, sortPriority=36}.R) {name=TextureOpactiy, nodeColor=<rgba 1 0.2578 0 1>}
  n19 = NamedRerouteUsage()
  n20 = ScalarParameter() {defaultValue=1, group=00_Glass, parameterName=TextureOpacityAdd, sortPriority=34}
  n21 = Saturate(None=Add(A=Multiply(A=@n19, B=@n20), B=NamedRerouteUsage()))
  n22 = ScalarParameter() {group=00_Glass, parameterName=TextureOpacityBlend, sortPriority=35}
  NamedRerouteDeclaration(None=StaticSwitchParameter(True=LinearInterpolate(A=@n21, B=@n19, Alpha=@n22), False=NamedRerouteUsage()) {group=00_Glass, parameterName=UseTextureOpacity, sortPriority=36}) {name=GlassOpacity, nodeColor=<rgba 0.9922 0 1 1>}
 ]}
