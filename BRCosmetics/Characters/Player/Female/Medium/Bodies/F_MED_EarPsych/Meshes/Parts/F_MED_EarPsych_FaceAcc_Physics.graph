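# Physics asset for the F_MED_EarPsych female character's face accessory part, defining collision and joints for left and right dynamic hair chains.
# Each side chains three capsule segments from a small base body, with two capsules on each middle segment.
# Constraints set zero twist limits and narrow swing limits, with one swing axis tightening toward the tip.
# Head, neck, clavicle, upper spine and upper arm bodies supply unconstrained collision shapes around the chains.
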PhysicsAsset /BRCosmetics/Characters/Player/Female/Medium/Bodies/F_MED_EarPsych/Meshes/Parts/F_MED_EarPsych_FaceAcc_Physics
bodies {
 base_dyn_hair_l [{shapeType=Capsule, length=0.5, radius=0.5, shapeName=base_dyn_hair_l_capsule, rotation=<rot 0 -90 0>}]
 base_dyn_hair_r [{shapeType=Capsule, length=0.5, radius=0.5, shapeName=base_dyn_hair_r_capsule, rotation=<rot 0 -90 0>}]
 clavicle_l [{shapeType=Capsule, length=12, radius=5.25, shapeName=clavicle_l_capsule, center=<3.5 -1 0>, rotation=<rot 80 -0.0003 -0.0003>}]
 clavicle_r [{shapeType=Capsule, length=12, radius=5.25, shapeName=clavicle_r_capsule, center=<-3.5 1 0>, rotation=<rot 80 -0.0003 -0.0004>}]
 dyn_hair_l_00 [{shapeType=Capsule, length=5.25, radius=2.25, shapeName=dyn_hair_l_00_capsule, center=<-0.5 0 0>, rotation=<rot 0 -90 90>}]
 dyn_hair_l_01 [
  {shapeType=Capsule, length=4.5, radius=2, shapeName=dyn_hair_l_01_capsule, center=<1.75 0.1 -0.1>, rotation=<rot 0 90 -90>}
  {shapeType=Capsule, length=3, radius=1, shapeName=dyn_hair_l_01_capsule, center=<1.5 -2 0>, rotation=<rot -6 60 -105.75>}
 ]
 dyn_hair_l_02 [{shapeType=Capsule, length=5.5, radius=1.5, shapeName=dyn_hair_l_02_capsule, center=<4.5 0 0>, rotation=<rot 0 -90 -90>}]
 dyn_hair_r_00 [{shapeType=Capsule, length=5.25, radius=2.25, shapeName=dyn_hair_r_00_capsule, center=<0.5 0 0>, rotation=<rot 0 -90 -90>}]
 dyn_hair_r_01 [
  {shapeType=Capsule, length=4.5, radius=2, shapeName=dyn_hair_r_01_capsule, center=<1.75 0.1 -0.1>, rotation=<rot 0 90 90>}
  {
   shapeType=Capsule
   length=5.2863
   radius=0.7855
   shapeName=dyn_hair_r_01_capsule
   center=<1.9033 -1.7061 0.7261>
   rotation=<rot -0.2843 86.8973 91.4914>
  }
 ]
 dyn_hair_r_02 [{shapeType=Capsule, length=5.5, radius=1.5, shapeName=dyn_hair_r_02_capsule, center=<4.5 0 -0.25>, rotation=<rot 0 -90 90>}]
 head [{shapeType=Sphere, radius=7.2087, shapeName=head_sphere, center=<0.0947 2.8266 0>}]
 neck_01 [{shapeType=Capsule, length=25, radius=2.15, shapeName=neck_01_capsule, center=<-1 4 -0.6>}]
 neck_02 [{shapeType=Capsule, length=4.4184, radius=3.4403, rotation=<rot 90 0 0>}]
 spine_05 [{shapeType=Capsule, length=18.6627, radius=8.1162, center=<0.2 2.25 0>}]
 upperarm_l [{shapeType=Capsule, length=29.25, radius=4.5, center=<12 0 0>, rotation=<rot 90 0 0>}]
 upperarm_r [{shapeType=Capsule, length=29.25, radius=4.5, center=<-12 0 0>, rotation=<rot 90 0 0>}]
}
constraints [
 {bone1Name=dyn_hair_l_00, bone2Name=base_dyn_hair_l, swing1LimitDegrees=15, swing2LimitDegrees=45, twistLimitDegrees=0}
 {bone1Name=dyn_hair_l_01, bone2Name=dyn_hair_l_00, swing1LimitDegrees=15, swing2LimitDegrees=20, twistLimitDegrees=0}
 {bone1Name=dyn_hair_l_02, bone2Name=dyn_hair_l_01, swing1LimitDegrees=15, swing2LimitDegrees=15, twistLimitDegrees=0}
 {bone1Name=dyn_hair_r_00, bone2Name=base_dyn_hair_r, swing1LimitDegrees=15, swing2LimitDegrees=45, twistLimitDegrees=0}
 {bone1Name=dyn_hair_r_01, bone2Name=dyn_hair_r_00, swing1LimitDegrees=15, swing2LimitDegrees=20, twistLimitDegrees=0}
 {bone1Name=dyn_hair_r_02, bone2Name=dyn_hair_r_01, swing1LimitDegrees=15, swing2LimitDegrees=15, twistLimitDegrees=0}
]
